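# Ambient sound cue for the Hightower HC biome, playing an idle vehicle interior loop for the QJ interior.
# A single looping wave player feeds the output directly on the Athena ambient sound class with a slight volume boost.
# Overridden attenuation uses a box shape with listener focus and interpolated occlusion that applies a low-pass filter and a volume multiplier above unity.
# The cue's debug flag is left on.
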SoundCue /Game/Sounds/Apollo/Biomes/Hightower_HC/AMB_Hightower_HC_QJ_Interior_Cue
root=WavePlayer_0
nodes [{
  id=WavePlayer_0
  properties {bLooping=true, soundWaveAssetPtr=&/Veh_Hightower_QJ_Idle_Interior_Loop_01}
 }]
properties {
 bDebug=true
 bOverrideAttenuation=true
 maxDistance=964.5126
 soundClassObject=&Fort_Ambient_Athena
 volumeMultiplier=1.253
 attenuationOverrides {
  attenuationShape=Box
  bEnableFocusInterpolation=true
  bEnableListenerFocus=true
  bEnableOcclusion=true
  falloffDistance=512
  nonFocusVolumeAttenuation=0.6
  occlusionInterpolationTime=0.5
  occlusionLowPassFilterFrequency=500
  occlusionTraceChannel=ECC_GameTraceChannel11
  occlusionVolumeAttenuation=2
  stereoSpread=2048
  attenuationShapeExtents=<147.3867 452.5126 125.0106>
 }
}
